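; Diving board prop (Dam variant) that bounces pawns and hoverboards touching its Touchbox.
; For a FortPawn, an additive launch cancels vertical velocity for a fixed upward bounce and pushes along world Y from current velocity.
; It then plays the board sound, skips the next fall damage and sends an Athena diving board quest stat.
; Other actors go through LaunchHoverboard, playing only the sound on success.
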(event OnComponentBeginOverlap(Touchbox)
  (params (OverlappedComponent :type Primitive%20Component%20Object%20Reference) (OtherActor :type Actor%20Object%20Reference) (OtherComp :type Primitive%20Component%20Object%20Reference) (OtherBodyIndex :type Integer) (bFromSweep :type Boolean) (SweepResult :type Hit%20Result%20Structure%20%28by%20ref%29))
  (bind _asfort_pawn (Utilities|Casting|CastToFortPawn OtherActor)
    (:then
      (Character|LaunchCharacter _asfort_pawn (Math|Vector|MakeVector 0.0 (+ (.y (Transformation|GetVelocity _asfort_pawn)) -5e+02) (+ (* (.z (Transformation|GetVelocity _asfort_pawn)) -1.0) 5e+02)) false false)
      (CallFunction|PlayReplicatedDBSFX)
      (Variables|/Script/FortniteGame.FortPawn|SetbIgnoreNextFallingDamage true _asfort_pawn)
      ;[1 1 1] Complex Custom Event for Diving Board Quests
      (Quest|SendComplexCustomStatEvent (Quests|GetQuestManager (Fortnite|GetFortPlayerControllerfromActor _asfort_pawn) "Athena") 0 "(GameplayTags=((TagName=\"Athena.Quests.DivingBoard\")))" "(GameplayTags=)" 1))
    (:CastFailed
      (bind _success (Vehicles|LaunchHoverboard OtherActor))
      (if _success
        (CallFunction|PlayReplicatedDBSFX)))))

(event Custom|Play%20Replicated%20DB%20SFX
  (Audio|PlaySoundatLocation (Variables|Default|GetSound) (Transformation|GetActorLocation)))

(fn PlayReplicatedDivingBoardSFX)

(fn ConstructionScript
  (|CallParentFunction))
